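# Head animation Blueprint for the FlavorStock character cosmetic, with graph logic to toggle a material element by style variant.
# The logic reads the owning pawn's Parts variant index and sets "Hide Element 01" on the part mesh's dynamic material, hiding it for index 0 (Stage1).
# This chain starts at a Sequence with no incoming event link; the update event only calls its parent, and OnCharacterCustomizationLoaded is empty.
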AnimBlueprint /BRCosmetics/Characters/Player/Male/Large/Heads/M_LRG_FlavorStock_Head/Meshes/M_LRG_FlavorStock_Head_AnimBP
target_skeleton=/Game/Characters/Player/Common/Fortnite_Base_Head/Fortnite_Base_Head_Export_Skeleton
graphs [EventGraph {nodes [
   n1 = Event BlueprintUpdateAnimation()
   > Parent: BlueprintUpdateAnimation!(execute, DeltaTimeX:float=@n1.DeltaTimeX) -> (then) {is=title K2Node_CallParentFunction}
   Sequence()
   > n2 = Cast To FortPlayerPawn(Object=GetOwningActor())
   > n3 = GetVariantChannelActiveIndexForItemActor!(
    execute
    Item:Athena Cosmetic Item Definition Object Reference=/BRCosmetics/Athena/Items/Cosmetics/Characters/Character_FlavorStock
    Actor:Actor Object Reference=@n2
    Channel:Gameplay Tag Structure=(TagName="Cosmetics.Variant.Channel.Parts")
    VariantList:Array of Gameplay Tag Structures=Make Array!("[0]"=(TagName="Cosmetics.Variant.Property.Stage1"), "[1]":Gameplay Tag Structure=(TagName="Cosmetics.Variant.Property.Stage2")) -> (Array)
    bErrorIfChannelMissing:bool=false
    bErrorIfTagNotInArray:bool=false
   ) -> (then, Index:int=0, bIndexIsValid:bool, bFoundChannel:bool)
   > n4 = Get Dynamic Material Instance(SkeletalMeshComponent=GetPartSkeletalMeshComponent())
   > n5 = Branch(Condition="Equal (Integer)"(A=@n3))
   > SetScalarParameterValue(self=@n4, ParameterName=Hide Element 01, Value=1)
   SetScalarParameterValue(execute=@n5.else, self=@n4, ParameterName=Hide Element 01, Value=0)
   TryGetPawnOwner()
   Event OnCharacterCustomizationLoaded()
  ]}]
variables []
